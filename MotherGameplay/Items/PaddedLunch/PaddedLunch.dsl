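; PaddedLunch is a Fortnite item actor blueprint that starts its ambient sound and applies gameplay effects when it spawns.
; BeginPlay fades in the ambient audio component, runs the parent logic, then applies its configured effect list to itself at its own location, using its own ability system component.
; The overlap event is an empty stub, and the construction script only calls the parent.
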(event EventBeginPlay
  (Audio|Components|Audio|FadeIn (Variables|Default|GetAmbientAudio) 0.25 1.0 0.0 "Linear" :!class /Script/Engine.AudioComponent)
  (|CallParentFunction)
  (Fortnite|ApplyGameplayEffectsToActor self (Variables|Default|GetEffects%20to%20Apply) (Transformation|GetActorLocation) false 0 (Ability|GetAbilitySystemComponent self)))

(event Collision|EventActorBeginOverlap
  (params (OtherActor :type Actor%20Object%20Reference)))

(fn ConstructionScript
  (|CallParentFunction))
